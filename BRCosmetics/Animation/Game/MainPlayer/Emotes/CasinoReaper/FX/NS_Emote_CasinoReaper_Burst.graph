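# Niagara burst for the CasinoReaper emote, combining a brief red muzzle-style flash with four flying card-suit sprites.
# Flash_Elements feeds spike, star, flare and light renderers from per-element particle attributes, using spawn-group visibility tags and smaller front-end sizes.
# NS_CardSuits picks subUV frame and launch position by spawn index, then applies gravity, drag and size curves.
# Time-of-day color curves tint flash and suits from white toward red, with dark alternate suits.
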NiagaraSystem /BRCosmetics/Animation/Game/MainPlayer/Emotes/CasinoReaper/FX/NS_Emote_CasinoReaper_Burst
dependencies {
 usedDataInterfaces [
  &NiagaraDataInterfaceSpriteRendererInfo
  &NiagaraDataInterfaceArrayFloat
  &NiagaraDataInterfaceCurve
  &NiagaraDataInterfaceColorCurve
 ]
 usedDynamicInputs []
 usedModules [&SystemState]
}
emitters {
 Flash_Elements {
  emitterUpdateScript [
   &EmitterState(Life Cycle Mode=::NewEnumerator1, Loop Behavior=::NewEnumerator1, Loop Duration=0.06, UseLoopDelay=true)
   &SpawnBurst_Instantaneous(Spawn Group=1, Use Spawn Probability=false, UseLoopCountLimit=false) {module=SpawnBurst_Instantaneous001}
   &SpawnBurst_Instantaneous(
    Spawn Count=&RandomRangeInt(Minimum=6, Maximum=8)
    Use Spawn Probability=false
    UseLoopCountLimit=false
   )
  ]
  particleSpawnScript [
   &InitializeParticle(Color Mode=::NewEnumerator0, Lifetime=0.25, UsePositionOffset=false, Write Material Random=true)
   &ShapeLocation(MaskSpawnGroup=false, Override Seed=false, Sphere Radius=1, UseOwnerScale=true)
   &SpriteFacingAndAlignment(Sprite Alignment=$Particles.ShapeLocation.ShapeNormal, WriteAlignment=true, WriteFacing=false)
   &FN_PerParticleLOD(Enable SpriteSize Distance Scaling=true, Sprite Size Scale Far=<2.5 2.5>)
   /SetVariables(Particles.SpriteSize=<1 1>, Particles.VisibilityTag=&SpawnGroup()) {isSetParameters=true}
   /SetVariables(
    Particles.Flash Spike Lifetime:NiagaraFloat=&RandomRangeFloat(Minimum=0.05, Maximum=0.07)
    Particles.Flash Spike Sprite Size:Vector2f=&FN_FrontEndVector2Modifier(
     Front End:Vector2f=&RandomRangeVector2D(Minimum=<2 25>, Maximum=<7 40>)
     Gameplay:Vector2f=&RandomRangeVector2D(Minimum=<3 55>, Maximum=<10 70>)
    )
   ) {isSetParameters=true}
   /SetVariables(
    Particles.Flash Star Cam Offset=15
    Particles.Flash Star Lifetime=0.15
    Particles.Flash Star Rotation:NiagaraFloat=&RandomRangeFloat(Minimum=0, Maximum=360)
    Particles.Flash Star Sprite Size:Vector2f=&Multiply_Vector2D_ByFloat(
     Float=&Multiply_Float(
      A=&MakeCustomFloatFromBool(Bool=$User.bIsFrontEnd, True Float=0.45, False Float=1)
      B=&RandomRangeFloat(Minimum=70, Maximum=100)
     )
     Vector2D=$Particles.SpriteSize
    )
    Particles.Flash Star Sprite Size Alt:Vector2f=&Multiply_Vector2D_ByFloat(
     Float=&Multiply_Float(
      A=0.85
      B=&Multiply_Float(
       A=&MakeCustomFloatFromBool(Bool=$User.bIsFrontEnd, True Float=0.45, False Float=1)
       B=&RandomRangeFloat(Minimum=70, Maximum=100)
      )
     )
     Vector2D=$Particles.SpriteSize
    )
    Particles.Flash Star SubUV Index:NiagaraFloat=&RandomRangeFloat(Minimum=0, Maximum=3)
    Particles.Flash Star SubUV Index Alt:NiagaraFloat=&RandomRangeFloat(Minimum=0, Maximum=3)
   ) {isSetParameters=true}
   /SetVariables(
    Particles.Flash Flare Color:LinearColor=<rgba 1 0 0 1>
    Particles.Flash Flare Lifetime:NiagaraFloat=&FN_FrontEndFloatModifier(Gameplay=0.2, Front End=0.125)
    Particles.Flash Flare Sprite Size:Vector2f=&FN_FrontEndVector2Modifier(Front End:Vector2f=<90 90>, Gameplay:Vector2f=<160 160>)
   ) {isSetParameters=true}
   &DynamicMaterialParameters(
    Index 0 Param 1=0.2
    Index 0 Param 2=1
    Module.Index0Param0Write=true
    Module.Index0Param1Write=true
    Module.Index0Param2Write=true
    Module.Index0Param3Write=true
   )
  ]
  particleUpdateScript [
   &ParticleState()
   /SetVariables(
    Particles.Update Flash Flare Age:NiagaraFloat=&Divide_Float(A=$Particles.Age, B=$Particles.Initial.Flash Flare Lifetime)
    Particles.Update Flash Spike Age:NiagaraFloat=&Divide_Float(A=$Particles.Age, B=$Particles.Initial.Flash Spike Lifetime)
    Particles.Update Flash Star Age:NiagaraFloat=&Divide_Float(A=$Particles.Age, B=$Particles.Initial.Flash Star Lifetime)
   ) {isSetParameters=true}
   /SetVariables(
    Particles.Update Flash Spike Color:LinearColor=&MakeLinearColorFromVectorAndFloat(
     "Vector (RGB)"=&MakeVectorFromLinearColorRGB(Color=$Particles.Initial.Flash Spike Color)
     "Float (Alpha)"=&FloatFromCurve(FloatCurve=DataInterface() {propertyValues {Curve {keys [<at 0.7 1>,<at 1 0>]}}}, CurveIndex=$Particles.Update Flash Spike Age)
    )
    Particles.Update Flash Spike Sprite Size:Vector2f=&Multiply_Vector2D_ByFloat(
     Float=&FloatFromCurve(FloatCurve=DataInterface() {propertyValues {Curve {keys [<at 0 0.6>,<at 1 2>]}}}, CurveIndex=$Particles.Update Flash Spike Age)
     Vector2D=$Particles.Initial.Flash Spike Sprite Size
    )
    Particles.Update Flash Star Color:LinearColor=&FNTimeOfDayColor(
     Night Color=&ColorFromCurve(
      ColorCurve=DataInterface() {propertyValues {
        AlphaCurve {keys [<at 0.3622 0.7>,<at 0.8992 0>]}
        BlueCurve {keys [<at 0 0.8281>,<at 0.1766 0.0394>]}
        GreenCurve {keys [<at 0 0.8028>,<at 0.1766 0>]}
        RedCurve {keys [<at 0 0.7366>,<at 0.1766 0.6476>]}
       }}
      CurveIndex=$Particles.Update Flash Star Age
     )
     Morning Color=&ColorFromCurve(
      ColorCurve=DataInterface() {propertyValues {
        AlphaCurve {keys [<at 0.3622 0.7>,<at 0.8992 0>]}
        BlueCurve {keys [<at 0 0.8281>,<at 0.1766 0.0247>]}
        GreenCurve {keys [<at 0 0.8028>,<at 0.1766 0.0521>]}
        RedCurve {keys [<at 0 0.7366>,<at 0.1766 0.7008>]}
       }}
      CurveIndex=$Particles.Update Flash Star Age
     )
     Day Color=&ColorFromCurve(
      ColorCurve=DataInterface() {propertyValues {
        AlphaCurve {keys [<at 0.3622 1>,<at 0.8992 0>]}
        BlueCurve {keys [<at 0 0.8281>,<at 0.1766 0>]}
        GreenCurve {keys [<at 0 0.8028>,<at 0.1766 0>]}
        RedCurve {keys [<at 0 0.7366>,<at 0.1766 1>]}
       }}
      CurveIndex=$Particles.Update Flash Star Age
     )
     Evening Color=&ColorFromCurve(
      ColorCurve=DataInterface() {propertyValues {
        AlphaCurve {keys [<at 0.3622 0.7>,<at 0.8992 0>]}
        BlueCurve {keys [<at 0 0.8281>,<at 0.1766 0.0719>]}
        GreenCurve {keys [<at 0 0.8028>,<at 0.1766 0.0077>]}
        RedCurve {keys [<at 0 0.7366>,<at 0.1766 0.636>]}
       }}
      CurveIndex=$Particles.Update Flash Star Age
     )
    )
    Particles.Update Flash Star Sprize Size:Vector2f=&Multiply_Vector2D_ByFloat(
     Float=&FloatFromCurve(
      FloatCurve=DataInterface() {propertyValues {Curve {keys [
          {arriveTangent=2.7013, interpMode=RCIM_Cubic, leaveTangent=2.7013, tangentMode=RCTM_User, time=0, value=0}
          {interpMode=RCIM_Cubic, tangentMode=RCTM_User, time=1, value=1}
         ]}}}
      CurveIndex=$Particles.Update Flash Star Age
     )
     Vector2D=$Particles.Initial.Flash Star Sprite Size
    )
    Particles.Update Flash Star Sprize Size Alt:Vector2f=&Multiply_Vector2D_ByFloat(
     Float=&FloatFromCurve(
      FloatCurve=DataInterface() {propertyValues {Curve {keys [
          {arriveTangent=2.7013, interpMode=RCIM_Cubic, leaveTangent=2.7013, tangentMode=RCTM_User, time=0, value=0}
          {interpMode=RCIM_Cubic, tangentMode=RCTM_User, time=1, value=1}
         ]}}}
      CurveIndex=$Particles.Update Flash Star Age
     )
     Vector2D=$Particles.Initial.Flash Star Sprite Size Alt
    )
   ) {isSetParameters=true}
   /SetVariables(Particles.Update Flash Flare Color:LinearColor=&MakeLinearColorFromVectorAndFloat(
     "Vector (RGB)"=&Multiply_Vector(
      A=&MakeVectorFromLinearColorRGB(Color=&FNTimeOfDayColor(
        Night Color=&ColorFromCurve(ColorCurve=DataInterface() {propertyValues {
           AlphaCurve {keys [<at 0 1 RCIM_Cubic>]}
           BlueCurve {keys [<at 0 0.8281 RCIM_Cubic>,<at 0.3389 0.0719 RCIM_Cubic>]}
           GreenCurve {keys [<at 0 0.8028 RCIM_Cubic>,<at 0.3389 0.0077 RCIM_Cubic>]}
           RedCurve {keys [<at 0 0.7366 RCIM_Cubic>,<at 0.3389 0.636 RCIM_Cubic>]}
          }})
        Morning Color=&ColorFromCurve(ColorCurve=DataInterface() {propertyValues {
           AlphaCurve {keys [<at 0 1 RCIM_Cubic>]}
           BlueCurve {keys [<at 0 0.8281 RCIM_Cubic>,<at 0.3389 0.0333 RCIM_Cubic>]}
           GreenCurve {keys [<at 0 0.8028 RCIM_Cubic>,<at 0.3389 0.0702 RCIM_Cubic>]}
           RedCurve {keys [<at 0 0.7366 RCIM_Cubic>,<at 0.3389 0.9439 RCIM_Cubic>]}
          }})
        Day Color=&ColorFromCurve(ColorCurve=DataInterface() {propertyValues {
           AlphaCurve {keys [<at 0 1 RCIM_Cubic>]}
           BlueCurve {keys [<at 0 0.8281 RCIM_Cubic>,<at 0.3389 0 RCIM_Cubic>]}
           GreenCurve {keys [<at 0 0.8028 RCIM_Cubic>,<at 0.3389 0 RCIM_Cubic>]}
           RedCurve {keys [<at 0 0.7366 RCIM_Cubic>,<at 0.3389 1 RCIM_Cubic>]}
          }})
        Evening Color=&ColorFromCurve(ColorCurve=DataInterface() {propertyValues {
           AlphaCurve {keys [<at 0 1 RCIM_Cubic>]}
           BlueCurve {keys [<at 0 0.8281 RCIM_Cubic>,<at 0.3389 0.0394 RCIM_Cubic>]}
           GreenCurve {keys [<at 0 0.8028 RCIM_Cubic>,<at 0.3389 0 RCIM_Cubic>]}
           RedCurve {keys [<at 0 0.7366 RCIM_Cubic>,<at 0.3389 0.6476 RCIM_Cubic>]}
          }})
       ))
      B=<1 1 1>
     )
     "Float (Alpha)"=&Multiply_Float(
      A=&MakeFloatFromLinearColor(LinearColor=<rgba 1 1 1 1>, Channel:ENiagaraLinearColor_Channels=::NewEnumerator3)
      B=&Lerp_Float(A=1, B=0, Alpha=$Particles.Update Flash Flare Age)
     )
    )) {isSetParameters=true}
   &FN_Light_Attributes(
    Color=<rgba 1 0 0.0061 1>
    Create Color Attribute=true
    Create Exponent Attribute=true
    Create Position Attribute=true
    Create Radius Attribute=true
    Exponent=5
    Position=<15 0 0>
    Radius=90
    Scale Color=&FloatFromCurve(
     FloatCurve=DataInterface() {propertyValues {Curve {keys [
         {interpMode=RCIM_Cubic, tangentMode=RCTM_User, time=0, value=1}
         {arriveTangent=-0.5426, interpMode=RCIM_Cubic, leaveTangent=-0.5426, tangentMode=RCTM_User, time=1, value=0}
        ]}}}
     Scale Curve=10
    )
   )
  ]
  rendererClasses [&NiagaraSpriteRendererProperties, &NiagaraLightRendererProperties]
  renderers [
   {rendererClass=&NiagaraSpriteRendererProperties}
   {rendererClass=&NiagaraSpriteRendererProperties, rendererIndex=1}
   {rendererClass=&NiagaraSpriteRendererProperties, rendererIndex=2}
   {rendererClass=&NiagaraSpriteRendererProperties, rendererIndex=3}
   {rendererClass=&NiagaraLightRendererProperties, rendererIndex=4}
  ]
 }
 NS_CardSuits {
  emitterUpdateScript [
   &EmitterState(Life Cycle Mode=::NewEnumerator1, Loop Behavior=::NewEnumerator1, Loop Duration=1, UseLoopDelay=false)
   &SpawnBurst_Instantaneous(Spawn Count=4, Use Spawn Probability=false, UseLoopCountLimit=false)
  ]
  particleSpawnScript [
   &InitializeParticle(
    Lifetime Max=0.2
    Lifetime Min=0.3
    Lifetime Mode=::NewEnumerator1
    Mass Mode=::NewEnumerator3
    Sprite Rotation Mode=::NewEnumerator2
    Sprite Size Mode=::NewEnumerator1
    Uniform Sprite Size Max=9
    UsePositionOffset=false
    Write Material Random=true
   )
   &SubUVAnimation(
    Direct Frame Index:NiagaraFloat=&SelectFloatFromArray(
     Direct Array Index=&ReturnExecIndex()
     Direct Array Mode=::NewEnumerator1
     Float Selection Array=DataInterface() {propertyValues {FloatData [0,1,2,3]}}
    )
    "Sprite Renderer "=DataInterface() {propertyValues {SpriteRenderer=&self:NS_CardSuits_0.NiagaraSpriteRendererProperties_0}}
    SubUV Animation Mode:ENiagara_SubUVLookupModeV2=::NewEnumerator5
   )
   &ShapeLocation(
    MaskSpawnGroup=false
    Sphere Distribution=::NewEnumerator1
    Sphere Radius=20
    U Position=&Multiply_Float(
     A=&SelectFloatFromArray(
      Direct Array Index=&ReturnExecIndex()
      Float Selection Array=DataInterface() {propertyValues {FloatData [0,0.4295,0.5885,-0.0117]}}
     )
     B=&RandomRangeFloat(Minimum=1.1, Maximum=0.9)
    )
    UseOwnerScale=true
    V Position=&Multiply_Float(
     A=&SelectFloatFromArray(
      Direct Array Index=&ReturnExecIndex()
      Float Selection Array=DataInterface() {propertyValues {FloatData [0.3396,0.3496,0.2697,0.8174]}}
     )
     B=&RandomRangeFloat(Minimum=1.1, Maximum=0.9)
    )
   )
   &AddVelocity(
    Module.Use Falloff Distance=false
    Origin Offset=<0 0 -10>
    Override Seed=false
    Velocity Mode=::NewEnumerator1
    Velocity Speed=&RandomRangeFloat(Minimum=300, Maximum=200)
   )
  ]
  particleUpdateScript [
   &ParticleState()
   &ScaleSpriteSize(Uniform Curve Sprite Scale=DataInterface() {propertyValues {Curve {keys [
        {interpMode=RCIM_Cubic, tangentMode=RCTM_User, time=0, value=0.5625}
        {interpMode=RCIM_Cubic, tangentMode=RCTM_User, time=0.0903, value=1.5088}
        {arriveTangent=-4.3254, interpMode=RCIM_Cubic, leaveTangent=-4.3254, tangentMode=RCTM_User, time=1, value=0}
       ]}}})
   &GravityForce(Gravity=<0 0 -454.2905>)
   &Drag(
    Drag=&FloatFromCurve(
     FloatCurve=DataInterface() {propertyValues {Curve {keys [
         {arriveTangent=2.1355, interpMode=RCIM_Cubic, leaveTangent=2.1355, tangentMode=RCTM_User, time=0.0098, value=0}
         {arriveTangent=5.4197, interpMode=RCIM_Cubic, leaveTangent=5.4197, tangentMode=RCTM_User, time=0.2076, value=1}
        ]}}}
     Scale Curve=2
    )
    Use Linear Drag=true
    Use Rotational Drag=false
   )
   &ScaleColor(
    Scale Alpha=&FloatFromCurve(FloatCurve=DataInterface() {propertyValues {Curve {keys [<at -0.0046 1.0113 RCIM_Cubic>,<at 1.0011 0.002 RCIM_Cubic>]}}})
    ScaleA=true
    ScaleRGB=true
   ) {enabled=false}
   &SolveForcesAndVelocity(Clamp Velocity=false, Limit Acceleration=false)
   &self:CardColors(
    Color1:LinearColor=&ColorFromCurve(ColorCurve=DataInterface() {propertyValues {
       AlphaCurve {keys [<at 0 1 RCIM_Cubic>]}
       BlueCurve {keys [<at 0 1 RCIM_Cubic>,<at 0.1716 0.0318 RCIM_Cubic>]}
       GreenCurve {keys [<at 0 1 RCIM_Cubic>,<at 0.1716 0.0269 RCIM_Cubic>]}
       RedCurve {keys [<at 0 1 RCIM_Cubic>,<at 0.1716 0.0142 RCIM_Cubic>]}
      }})
    Color2:LinearColor=&FNTimeOfDayColor(
     Night Color=&ColorFromCurve(ColorCurve=DataInterface() {propertyValues {
        AlphaCurve {keys [<at 0 1 RCIM_Cubic>]}
        BlueCurve {keys [<at 0 1 RCIM_Cubic>,<at 0.1731 0.0394 RCIM_Cubic>]}
        GreenCurve {keys [<at 0 1 RCIM_Cubic>,<at 0.1731 0 RCIM_Cubic>]}
        RedCurve {keys [<at 0 1 RCIM_Cubic>,<at 0.1731 0.6476 RCIM_Cubic>]}
       }})
     Morning Color=&ColorFromCurve(ColorCurve=DataInterface() {propertyValues {
        AlphaCurve {keys [<at 0 1 RCIM_Cubic>]}
        BlueCurve {keys [<at 0 1 RCIM_Cubic>,<at 0.1731 0.0247 RCIM_Cubic>]}
        GreenCurve {keys [<at 0 1 RCIM_Cubic>,<at 0.1731 0.0521 RCIM_Cubic>]}
        RedCurve {keys [<at 0 1 RCIM_Cubic>,<at 0.1731 0.7008 RCIM_Cubic>]}
       }})
     Day Color=&ColorFromCurve(ColorCurve=DataInterface() {propertyValues {
        AlphaCurve {keys [<at 0 1 RCIM_Cubic>]}
        BlueCurve {keys [<at 0 1 RCIM_Cubic>,<at 0.1731 0 RCIM_Cubic>]}
        GreenCurve {keys [<at 0 1 RCIM_Cubic>,<at 0.1731 0 RCIM_Cubic>]}
        RedCurve {keys [<at 0 1 RCIM_Cubic>,<at 0.1731 0.467 RCIM_Cubic>]}
       }})
     Evening Color=&ColorFromCurve(ColorCurve=DataInterface() {propertyValues {
        AlphaCurve {keys [<at 0 1 RCIM_Cubic>]}
        BlueCurve {keys [<at 0 1 RCIM_Cubic>,<at 0.1731 0.0719 RCIM_Cubic>]}
        GreenCurve {keys [<at 0 1 RCIM_Cubic>,<at 0.1731 0.0077 RCIM_Cubic>]}
        RedCurve {keys [<at 0 1 RCIM_Cubic>,<at 0.1731 0.636 RCIM_Cubic>]}
       }})
    )
    Selector=2
   ) {module=CardColors}
  ]
  rendererClasses [&NiagaraSpriteRendererProperties]
  renderers [{rendererClass=&NiagaraSpriteRendererProperties}]
 }
}
summary {
 systemName=NS_Emote_CasinoReaper_Burst
 emitters [
  {emitterName=NS_CardSuits, rendererClasses [&NiagaraSpriteRendererProperties]}
  {
   emitterName=Flash_Elements
   rendererClasses [&NiagaraSpriteRendererProperties, &NiagaraLightRendererProperties]
  }
 ]
 userVariables [
  User.bIsFrontEnd {type=NiagaraBool}
  User.bIsFrontEndPreview {type=NiagaraBool}
 ]
}
