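# Niagara system for the card despawn effect of the CasinoReaper main-player emote, with a single emitter named Despawns.
# An integer User.ColorIndex parameter is exposed to the caller.
# System spawn runs two Set Variables modules, and the system draws on float-array selection, color and float curves, and color lerp dynamic inputs.
# The emitter lists no renderers, and the wiring between ColorIndex and these inputs is not shown.
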NiagaraSystem /BRCosmetics/Animation/Game/MainPlayer/Emotes/CasinoReaper/FX/NS_Emote_CasinoReaper_CardDespawn
dependencies {
 usedDataInterfaces [&NiagaraDataInterfaceColorCurve, &NiagaraDataInterfaceCurve, &NiagaraDataInterfaceArrayFloat]
 usedDynamicInputs [
  &ColorFromCurve
  &FloatFromCurve
  &LerpLinearColors
  &SelectFloatFromArray
 ]
 usedModules [
  &self:SystemSpawnScript.SystemScriptSource.SystemScriptGraph.NiagaraNodeAssignment_0.SetVariables_#1
  &self:SystemSpawnScript.SystemScriptSource.SystemScriptGraph.NiagaraNodeAssignment_1.SetVariables_#2
  &SystemState
 ]
}
emitters {Despawns {rendererClasses [], renderers []}}
summary {
 systemName=NS_Emote_CasinoReaper_CardDespawn
 emitters [{emitterName=Despawns, rendererClasses []}]
 userVariables [User.ColorIndex {type=NiagaraInt32}]
}
