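# Niagara ribbon trail effect for the CasinoReaper emote.
# One emitter spawns short-lived particles at a high rate into a ribbon renderer, facing along the owner's Z axis, with a custom Leading_Edge_Fix module in update.
# Color and alpha fade over particle life through curves, tinted and dimmed by time-of-day color and scalar inputs.
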NiagaraSystem /BRCosmetics/Animation/Game/MainPlayer/Emotes/CasinoReaper/FX/NS_Emote_CasinoReaper_Trail
dependencies {
 usedDataInterfaces [&NiagaraDataInterfaceColorCurve, &NiagaraDataInterfaceCurve]
 usedDynamicInputs []
 usedModules [&SystemState]
}
emitters {Trail001 {
  emitterUpdateScript [
   &EmitterState(
    Inactive Response=::NewEnumerator1
    Life Cycle Mode=::NewEnumerator1
    Loop Behavior=::NewEnumerator1
    Loop Delay=$System.Delay
    Loop Duration=1
    UseLoopDelay=false
   )
   /SetVariables(Emitter.NewSpawnRate=300) {isSetParameters=true}
   &SpawnRate(SpawnRate=$Emitter.NewSpawnRate, Use Spawn Probability=false)
  ]
  particleSpawnScript [&InitializeParticle(
    Lifetime=0.15
    Ribbon Facing Mode=::NewEnumerator1
    Ribbon Facing Vector:Vector3f=$Engine.Owner.SystemZAxis
    Ribbon Width=73
    Ribbon Width Mode=::NewEnumerator1
    UsePositionOffset=false
    Write Material Random=true
   )]
  particleUpdateScript [
   &self:Leading_Edge_Fix() {module=Leading_Edge_Fix}
   &ParticleState()
   &ScaleColor(
    Scale Alpha=&FloatFromCurve(
     FloatCurve=DataInterface() {propertyValues {Curve {keys [
         {
          arriveTangent=-9.3325
          arriveTangentWeight=0.4147
          interpMode=RCIM_Cubic
          leaveTangent=-9.3325
          leaveTangentWeight=0.4147
          tangentMode=RCTM_User
          tangentWeightMode=RCTWM_WeightedBoth
          time=0
          value=1
         }
         {arriveTangent=-1, interpMode=RCIM_Cubic, leaveTangent=-1, time=0.2087, value=0.2586}
         {arriveTangent=-0.1704, interpMode=RCIM_Cubic, leaveTangent=-0.1704, tangentMode=RCTM_User, time=1, value=0}
        ]}}}
     Scale Curve=&Multiply_Float(A=1, B=&FNTimeOfDayScalar(Night Scale=0.7, Morning Scale=0.7, Evening Scale=0.7))
    )
    Scale RGB=&Multiply_Vector(
     A=&MakeVectorFromLinearColorRGB(Color=&ColorFromCurve(ColorCurve=DataInterface() {propertyValues {
         AlphaCurve {keys []}
         BlueCurve {keys [<at 0 0.8652 RCIM_Cubic>,<at 0.4085 0.4855>,<at 1 0.1383 RCIM_Cubic>]}
         GreenCurve {keys [<at 0 0.8526 RCIM_Cubic>,<at 0.4085 0.4311>,<at 1 0.1053 RCIM_Cubic>]}
         RedCurve {keys [<at 0 0.7936 RCIM_Cubic>,<at 0.4085 0.3689>,<at 1 0.0677 RCIM_Cubic>]}
        }}))
     B=&MakeVectorFromLinearColorRGB(Color=&FNTimeOfDayColor(
       Night Color=<rgba 0.3169 0.4621 1 1>
       Morning Color=<rgba 1 0.8374 0.7324 1>
       Day Color=<rgba 1 1 1 1>
       Evening Color=<rgba 0.736 0.9073 1 1>
      ))
    )
    ScaleA=true
    ScaleRGB=true
   )
   &SolveForcesAndVelocity(Clamp Velocity=false, Limit Acceleration=false)
  ]
  rendererClasses [&NiagaraRibbonRendererProperties]
  renderers [{rendererClass=&NiagaraRibbonRendererProperties}]
 }}
summary {
 systemName=NS_Emote_CasinoReaper_Trail
 emitters [{emitterName=Trail001, rendererClasses [&NiagaraRibbonRendererProperties]}]
 userVariables [User.RibbonIDOffset {type=NiagaraInt32}]
}
